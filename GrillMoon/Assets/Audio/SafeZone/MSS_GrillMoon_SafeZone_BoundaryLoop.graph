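# Stereo MetaSound source that plays the looping boundary sound for GrillMoon safe zones.
# It is a preset of the shared BasicLoop Stereo MetaSound, so the playback graph comes from that parent.
# It supplies the POI boundary loop wave and a master volume as inputs.
# A safe zone attenuation asset handles spatial falloff, and the sound routes through the Fort_Events sound class.
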MetaSoundSource /GrillMoon/Assets/Audio/SafeZone/MSS_GrillMoon_SafeZone_BoundaryLoop
preset_parent=/GrillMoon_Audio/Assets/MSS_GrillMoon_BasicLoop_Stereo
inputs [
 Master Volume {type=Float}
 UE.Source.OnPlay {type=Trigger}
 UE.Source.StartTime.StartTime {type=Time}
 Wave Asset {type=WaveAsset}
]
interfaces [UE.OutputFormat.Stereo:1.0, UE.Source.StartTime:1.0, UE.Source:1.1]
outputs [
 "UE.OutputFormat.Stereo.Audio:0" {type=Audio}
 "UE.OutputFormat.Stereo.Audio:1" {type=Audio}
]
pages [Default {input_defaults {
   Master Volume=0.75
   UE.Source.OnPlay=null
   UE.Source.StartTime.StartTime=null
   Wave Asset=/GrillMoon/Assets/Audio/wav/GrillMoon_POI_Boundary_Loop
  }}]
properties {
 assetImportData=&self:AssetImportData
 attenuationSettings=/GrillMoon_Audio/Assets/Settings/ATT_GrillMoon_POI_SafeZone
 bLooping=true
 duration=10000
 maxDistance=1000
 numChannels=2
 outputFormat=Stereo
 sampleRate=48000
 soundClassObject=&Fort_Events
}
