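# MetaSound Source preset that plays the stereo one-shot sound for entering a GrillMoon safe zone, judging by its name and POI boundary-enter waves.
# It inherits its graph from MSS_GrillMoon_OneShotArray_Stereo, feeding it two wave variations and a master volume.
# Very-close POI attenuation, the gameplay-events sound class and a concurrency override that caps instances and retrigger rate govern playback.
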MetaSoundSource /GrillMoon/Assets/Audio/SafeZone/MSS_GrillMoon_SafeZone_Enter
preset_parent=/GrillMoon_Audio/Assets/MSS_GrillMoon_OneShotArray_Stereo
inputs [
 Master Volume {type=Float}
 Sounds {type=WaveAsset:Array}
 UE.Source.OnPlay {type=Trigger}
 UE.Source.StartTime.StartTime {type=Time}
]
interfaces [UE.OutputFormat.Stereo:1.0, UE.Source.OneShot:1.0, UE.Source.StartTime:1.0, UE.Source:1.1]
outputs [
 "UE.OutputFormat.Stereo.Audio:0" {type=Audio}
 "UE.OutputFormat.Stereo.Audio:1" {type=Audio}
 UE.Source.OneShot.OnFinished {type=Trigger}
]
pages [Default {input_defaults {
   Master Volume=1
   UE.Source.OnPlay=null
   UE.Source.StartTime.StartTime=null
   Sounds [/GrillMoon/Assets/Audio/wav/GrillMoon_POI_Boundary_Enter_01, /GrillMoon/Assets/Audio/wav/GrillMoon_POI_Boundary_Enter_02]
  }}]
properties {
 assetImportData=&self:AssetImportData
 attenuationSettings=/GrillMoon_Audio/Assets/Settings/ATT_GrillMoon_POI_VeryClose
 bOverrideConcurrency=true
 maxDistance=4100
 numChannels=2
 outputFormat=Stereo
 sampleRate=48000
 soundClassObject=&Fort_Events_Gameplay
 concurrencyOverrides {maxCount=3, retriggerTime=1}
}
